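# Animation Blueprint for the face accessory mesh of the F_MED_StreetOps female player body, adding secondary motion on top of the body's animation.
# It copies pose and curves from the attached parent mesh, then runs Anim Dynamics on dyn_zipper with an angular spring, limits and reduced gravity, relative to the head.
# A custom Flap node then offsets head_gullet along Y from the head bone's Z rotation.
# The event graph holds no logic.
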AnimBlueprint /BRCosmetics/Characters/Player/Female/Medium/Bodies/F_MED_StreetOps/Meshes/F_MED_Street_Ops_FaceAcc_AnimBP
target_skeleton=&/F_MED_Street_Ops_FaceAcc_Skeleton
graphs [
 AnimGraph {nodes [
   n1 = Anim Dynamics - Bone: dyn_zipper(ComponentPose=Local To Component(LocalPose=Copy Pose From Mesh() {bCopyCurves=true, bUseAttachedParent=true}), Alpha=1) {
    angularDampingOverride=0.99
    angularSpringConstant=250
    bAngularSpring=true
    bGravityOverrideInSimSpace=true
    gravityScale=0.33
    alphaBoolBlend {blendOption=Linear}
    boundBone {boneName=dyn_zipper}
    componentLinearAccScale=<1 1 1>
    componentLinearVelScale=<8 8 8>
    gravityOverride {z=-980}
    physicsBodyDefinitions [{
      sphereCollisionRadius=0
      boundBone {boneName=dyn_zipper}
      boxExtents=<10 10 10>
      constraintSetup {angularTargetAxis=AxisZ, twistAxis=AxisZ, angularLimitsMax=<45 0 0>, angularLimitsMin=<-45 0 0>, angularTarget=<0 0 0>}
      localJointOffset=<0 0 -3>
     }]
    relativeSpaceBone {boneName=head}
   }
   n2 = Flap Node - Bone: head_gullet!(ComponentPose:Component Space Pose Link Structure=@n1, Alpha:float=1) -> (Pose:Component Space Pose Link Structure) {
    is=title FortAnimGraphNode_Flap
    targetComponent=TranslationY
    alphaBoolBlend {blendOption=Linear}
    sourceDrives [{
      bUseRange=false
      drivingCurve=null
      multiplier=-15
      rangeMax=1
      rangeMin=-1
      remappedMax=1
      remappedMin=0
      sourceComponent=RotationZ
      useQuaternion=true
      sourceBone {boneName=head}
     }]
    targetBone {boneName=head_gullet}
   }
   Output Pose(Result=Component To Local(ComponentPose=@n2)) {layerGroup=None}
  ]}
 EventGraph {nodes [Event BlueprintUpdateAnimation()]}
]
variables []
